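# Physics asset for the AuraCop gadget prop, whose skeleton carries a poster with a five-joint paper chain.
# Each bone has one capsule: sizeable on paper joints 01, 02, 03 and 05, tiny on Paper_04, poster, main and offset, and very large-radius on root.
# Constraints link the bones in a single chain from root through offset, main and poster to the last paper joint, with uniform 45-degree swing and twist limits.
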PhysicsAsset /BRCosmetics/Animation/Props/Gadgets_AuraCop/Meshes/AuraCop_Gadget_Physics
bodies {
 JNT_Paper_01 [{shapeType=Capsule, length=18.8747, radius=9.5783, center=<0.1462 1.6193 -9.1394>, rotation=<rot 0 -90.7516 -89.2403>}]
 JNT_Paper_02 [{shapeType=Capsule, length=18.7662, radius=4.027, center=<0.2134 1.6203 -1.2032>, rotation=<rot 0 -90.9865 -88.4417>}]
 JNT_Paper_03 [{shapeType=Capsule, length=19.3227, radius=9.4874, center=<0.1924 0.6261 -1.8374>, rotation=<rot 0 89.1559 -86.6195>}]
 JNT_Paper_04 [{shapeType=Capsule, length=0.505, radius=0.505, rotation=<rot 0 -90 0>}]
 JNT_Paper_05 [{shapeType=Capsule, length=19.45, radius=9.3661, center=<-0.0714 0.6543 6.9901>, rotation=<rot 0 -90.9502 -82.7564>}]
 JNT_Poster [{shapeType=Capsule, length=0.505, radius=0.505, rotation=<rot 0 -90 0>}]
 main [{shapeType=Capsule, length=0.505, radius=0.505, rotation=<rot 0 -90 0>}]
 offset [{shapeType=Capsule, length=0.505, radius=0.505, rotation=<rot 0 -90 0>}]
 root [{shapeType=Capsule, length=0.5049, radius=374.7623, rotation=<rot 0 -90 0>}]
}
constraints [
 {bone1Name=offset, bone2Name=root, swing1LimitDegrees=45, swing2LimitDegrees=45, twistLimitDegrees=45}
 {bone1Name=main, bone2Name=offset, swing1LimitDegrees=45, swing2LimitDegrees=45, twistLimitDegrees=45}
 {bone1Name=JNT_Poster, bone2Name=main, swing1LimitDegrees=45, swing2LimitDegrees=45, twistLimitDegrees=45}
 {bone1Name=JNT_Paper_01, bone2Name=JNT_Poster, swing1LimitDegrees=45, swing2LimitDegrees=45, twistLimitDegrees=45}
 {bone1Name=JNT_Paper_02, bone2Name=JNT_Paper_01, swing1LimitDegrees=45, swing2LimitDegrees=45, twistLimitDegrees=45}
 {bone1Name=JNT_Paper_03, bone2Name=JNT_Paper_02, swing1LimitDegrees=45, swing2LimitDegrees=45, twistLimitDegrees=45}
 {bone1Name=JNT_Paper_04, bone2Name=JNT_Paper_03, swing1LimitDegrees=45, swing2LimitDegrees=45, twistLimitDegrees=45}
 {bone1Name=JNT_Paper_05, bone2Name=JNT_Paper_04, swing1LimitDegrees=45, swing2LimitDegrees=45, twistLimitDegrees=45}
]
